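# Sound cue for the MistRaven prototype gameplay cue audio, playing the speedlines loop end sound.
# A single wave player feeds a Modulator at default settings, which is the root.
# It uses a close-range stereo melee attack attenuation asset and the melee weapon sound class, with reduced volume.
# A concurrency override caps simultaneous instances.
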SoundCue /MistRaven/GameplayCues/Prototype/Audio/SC_MistRaven_Speedlines_lp_End
root=Modulator_0
nodes [
 {id=Modulator_0, children [WavePlayer_2]}
 {id=WavePlayer_2, properties {soundWaveAssetPtr=/MistRaven/GameplayCues/Prototype/Audio/wav/MistRaven_Speedlines_lp_End}}
]
properties {
 attenuationSettings=/MistRaven/GameplayCues/Prototype/Audio/Settings/CS_MistRaven_Melee_Attack_Close_Range_Stereo_Att
 bOverrideConcurrency=true
 maxDistance=7000
 soundClassObject=&Fort_WeaponMelee_PSM
 volumeMultiplier=0.7
 concurrencyOverrides {maxCount=3}
}
