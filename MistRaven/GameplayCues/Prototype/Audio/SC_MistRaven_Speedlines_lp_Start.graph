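# SoundCue for the start of MistRaven's looping speedlines effect audio, stored among prototype gameplay cue assets.
# A single wave player feeds a Modulator at the root, which is left at default settings.
# It uses a close-range stereo melee attack attenuation, plays in a melee weapon sound class with a slight volume boost, and overrides concurrency to cap simultaneous instances.
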SoundCue /MistRaven/GameplayCues/Prototype/Audio/SC_MistRaven_Speedlines_lp_Start
root=Modulator_0
nodes [
 {id=Modulator_0, children [WavePlayer_0]}
 {id=WavePlayer_0, properties {soundWaveAssetPtr=/MistRaven/GameplayCues/Prototype/Audio/wav/MistRaven_Speedlines_lp_Start}}
]
properties {
 attenuationSettings=/MistRaven/GameplayCues/Prototype/Audio/Settings/CS_MistRaven_Melee_Attack_Close_Range_Stereo_Att
 bOverrideConcurrency=true
 maxDistance=7000
 soundClassObject=&Fort_WeaponMelee_PSM
 volumeMultiplier=1.2
 concurrencyOverrides {maxCount=3}
}
