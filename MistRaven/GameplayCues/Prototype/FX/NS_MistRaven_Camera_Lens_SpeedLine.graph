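# Prototype Niagara system for a MistRaven gameplay cue, named as a camera-lens speed-line effect.
# It holds one emitter, LW_SpeedLine, with no renderers listed.
# The system uses SystemState, a spawn-time variable assignment and a bool-to-linear-color dynamic input.
# It exposes a User.Ascended bool, true by default, which likely selects the color.
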NiagaraSystem /MistRaven/GameplayCues/Prototype/FX/NS_MistRaven_Camera_Lens_SpeedLine
dependencies {
 usedDataInterfaces []
 usedDynamicInputs [&MakeCustomLinearColorFromBool]
 usedModules [
  &self:SystemSpawnScript.SystemScriptSource.SystemScriptGraph.NiagaraNodeAssignment_0.SetVariables_#1
  &SystemState
 ]
}
emitters {LW_SpeedLine {rendererClasses [], renderers []}}
summary {
 systemName=NS_MistRaven_Camera_Lens_SpeedLine
 emitters [{emitterName=LW_SpeedLine, rendererClasses []}]
 userVariables [User.Ascended {type=NiagaraBool, defaultValue=true}]
}
