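; Looping gameplay cue for the MistRaven super sprint (GrimStrut) that, despite its Speedlines name, scripts only audio feedback.
; Loop start attaches a 1P or 3P sound, shows a sound indicator for non-teammates, and activates a control bus mix for the viewed player unless hotfix-disabled.
; Application times a warning from duration and percentage, triggering the loop's ending burst plus a warning sound.
; Removal fades audio and stops the indicator and mix.
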(event EventOnLoopingStart(EitherVFX-preferthisversion)
  (params (MyTarget :type Actor%20Object%20Reference) (Parameters :type Gameplay%20Cue%20Parameters%20Structure%20%28by%20ref%29) (ParticleComponents :type Array%20of%20FXSystem%20Component%20Object%20References) (AudioComponents :type Array%20of%20Audio%20Component%20Object%20References))
  (bind _asfort_player_pawn_athena (Utilities|Casting|CastToFortPlayerPawnAthena MyTarget))
  (bind _returnvalue (Audio|SpawnSoundAttached (select (Pawn|IsLocallyViewed _asfort_player_pawn_athena) (Variables|Default|GetGrimStrutLoopSound_1P) (Variables|Default|GetGrimStrutLoopSound_3P)) (Variables|/Script/Engine.Actor|GetRootComponent _asfort_player_pawn_athena)))
  (Variables|Default|SetGrimStrutAudio _returnvalue)
  (if (not (Team|OnLocalPlayersTeam _asfort_player_pawn_athena _asfort_player_pawn_athena))
    (SoundIndicator|StartTargetedSoundIndicatorLooping (Variables|Default|GetFortSoundIndicator) _asfort_player_pawn_athena "/MistRaven/Assets/Audio/Temp_AudioVisualizer/MistRaven_Temp_Fire_Cue.MistRaven_Temp_Fire_Cue" 0.1 (Variables|/Script/Engine.Pawn|GetPlayerState _asfort_player_pawn_athena)))
  (if (ScalableFloat|GetValueasBool (Variables|Default|GetHFDeactivateGrimStrutCBM))
    _
    (elif (Pawn|IsLocallyViewed _asfort_player_pawn_athena)
      (Audio|Modulation|ActivateControlBusMix _asfort_player_pawn_athena (Variables|Default|GetGrimStrutCBM)))))

(event EventOnRemoval(EitherVFX-preferthisversion)
  (params (MyTarget :type Actor%20Object%20Reference) (Parameters :type Gameplay%20Cue%20Parameters%20Structure%20%28by%20ref%29) (ParticleComponents :type Array%20of%20FXSystem%20Component%20Object%20References) (AudioComponents :type Array%20of%20Audio%20Component%20Object%20References) (BurstCameraShakeInstance :type Camera%20Shake%20Base%20Object%20Reference) (BurstDecalInstance :type Decal%20Actor%20Object%20Reference))
  (Audio|Components|Audio|FadeOut (Variables|Default|GetGrimStrutAudio) 0.2 0.0 "Linear" :!class /Script/Engine.AudioComponent)
  (SoundIndicator|StopSoundIndicatorLooping (Variables|Default|GetFortSoundIndicator))
  (Audio|Modulation|DeactivateControlBusMix MyTarget (Variables|Default|GetGrimStrutCBM)))

(event EventOnApplication(EitherVFX-preferthisone)
  (params (MyTarget :type Actor%20Object%20Reference) (Parameters :type Gameplay%20Cue%20Parameters%20Structure%20%28by%20ref%29) (ParticleComponents :type Array%20of%20FXSystem%20Component%20Object%20References) (AudioComponents :type Array%20of%20Audio%20Component%20Object%20References) (BurstCameraShakeInstance :type Camera%20Shake%20Base%20Object%20Reference) (BurstDecalInstance :type Decal%20Actor%20Object%20Reference))
  (Variables|Default|SetMy%20Target MyTarget)
  (Utilities|Time|SetTimerbyEvent (EventDispatchers|CreateEvent :!event Custom|EndWarningBurst) (- (ScalableFloat|GetValueatLevel (Variables|Default|GetGrimStrutDuration)) (* (ScalableFloat|GetValueatLevel (Variables|Default|GetGrimStrutDuration)) (Variables|Default|GetWarningLevelPercent)))))

(event Custom|EndWarningBurst
  (Audio|Parameter|ExecuteTriggerParameter (Variables|Default|GetGrimStrutAudio) "EndingBurstTrigger")
  (Audio|SpawnSoundAttached (Variables|Default|GetEndWarningSound) (Variables|/Script/Engine.Actor|GetRootComponent (Variables|Default|GetMy%20Target))))

(fn ConstructionScript)
